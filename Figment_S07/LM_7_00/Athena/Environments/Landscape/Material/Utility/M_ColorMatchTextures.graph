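# Unlit utility material for color-matching landscape grass textures and deriving replacement tint values for farm, swamp and mountain grass.
# High-mip samples give average colors, whose ratio against the base Grasslands texture forms a color delta.
# Existing desaturate-and-tint results are divided by the delta-corrected textures, with DebugFloat3Values displaying swamp and mountain results.
# Separate branches preview NewTint parameters on the color-matched farm and base Grasslands textures.
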Material /Figment_S07/LM_7_00/Athena/Environments/Landscape/Material/Utility/M_ColorMatchTextures
graph {
 properties {blendMode=BLEND_AlphaComposite, pixelDepthOffsetMode=PDOM_Legacy, shadingModel=MSM_Unlit, usedShadingModels=Unlit}
 nodes [
  n1 = TextureSample!(UVs, Tex, Level=ScalarParameter() {parameterName=ml_5}, Apply View MipBias) {mipValueMode=TMVM_MipLevel, texture=/Figment_S07/LM_7_00/Athena/Environments/Landscape/Textures/T_Athena_Grasslands_Farm_D}
  n2 = Multiply(
   A=Desaturation(None=@n1.RGB, Fraction=ScalarParameter() {defaultValue=0.5, parameterName=Farm_Grass_Desat})
   B=VectorParameter() {parameterName=FarmGrass_Tint, defaultValue=<rgba 0.7637 1 0.1661 0>}.RGB
  )
  n3 = ScalarParameter() {defaultValue=12, parameterName=ml_1}
  n4 = Divide(
   A=TextureSample(Level=@n3) {mipValueMode=TMVM_MipLevel, texture=/Figment_S07/LM_7_00/Packages/DS_Fortnite_Terrain_NoLOD/Textures/T_Grasslands_D}.RGB
   B=TextureSample(Level=@n3) {mipValueMode=TMVM_MipLevel, texture=&T_Athena_Grasslands_Farm_D}.RGB
  )
  Divide(A=@n2, B=Multiply(A=@n4, B=@n1.RGB))
  n5 = ScalarParameter() {defaultValue=12, parameterName=ml_2}
  n6 = Divide(
   A=TextureSample(Level=@n5) {mipValueMode=TMVM_MipLevel, texture=&T_Grasslands_D}.RGB
   B=TextureSample(Level=@n5) {mipValueMode=TMVM_MipLevel, texture=&T_Athena_Grasslands_Farm_D}.RGB
  )
  Divide(B=@n6) {desc=Color Delta}
  n7 = ScalarParameter() {defaultValue=12, parameterName=ml}
  n8 = Divide(
   A=TextureSample(Level=@n7) {mipValueMode=TMVM_MipLevel, texture=&T_Grasslands_D}.RGB
   B=TextureSample(Level=@n7) {mipValueMode=TMVM_MipLevel, texture=&T_Grasslands_D}.RGB
  )
  Divide(B=@n8)
  Divide(B=@n4)
  n9 = TextureSample() {
   mipValueMode=TMVM_MipLevel
   samplerType=SAMPLERTYPE_LinearColor
   texture=/Figment_S07/LM_7_00/Athena/Environments/Landscape/Textures/T_Athena_Grass_Farm_ColorMatched_D
  }
  LinearInterpolate(
   A=Multiply(A=@n9.RGB, B=VectorParameter() {parameterName=NewTint1, defaultValue=<rgba 1.5953 1.022 0.5709 0>}.RGB)
   B=@n2
   Alpha=ScalarParameter() {parameterName=asdf}
  )
  n10 = TextureSample() {
   mipValueMode=TMVM_MipLevel
   samplerType=SAMPLERTYPE_LinearColor
   texture=&T_Athena_Grass_Farm_ColorMatched_D
  }
  LinearInterpolate(
   A=TextureSample() {mipValueMode=TMVM_MipLevel, texture=&T_Grasslands_D}.RGB
   B=@n10.RGB
   Alpha=ScalarParameter() {defaultValue=0.6}
  )
  n11 = TextureSample(Level=ScalarParameter() {parameterName=ml_3}) {mipValueMode=TMVM_MipLevel, texture=&T_Athena_Grasslands_Farm_D}
  n12 = Desaturation(None=@n11.RGB, Fraction=ScalarParameter() {parameterName=Swamp_Grass_Desat})
  n13 = VectorParameter() {parameterName=SwampGrass_Tint, defaultValue=<rgba 0.2502 0.3564 0.0595 0>}
  MaterialFunctionCall!(
   Vector3=Divide(A=Multiply(A=@n12, B=@n13.RGB), B=Multiply(A=@n6, B=@n11.RGB) {desc=ColorMatched}) {desc=New Color delta from Neutral}
   MaximumNumberOfDigits
   UVs
   DebugTextLocation RG_UpperRight BA_LowerLeft
   Component Spacing
  ) {materialFunction=/Engine/Functions/Engine_MaterialFunctions02/Utility/DebugFloat3Values}
  n14 = TextureSample(Level=ScalarParameter() {parameterName=ml_4}) {mipValueMode=TMVM_MipLevel, texture=&T_Grasslands_D}
  n15 = Desaturation(None=@n14.RGB, Fraction=ScalarParameter() {defaultValue=0.5, parameterName=Mountrain_Grass_Desat})
  n16 = VectorParameter() {parameterName=MountainGrass_Tint, defaultValue=<rgba 0.2937 0.5208 0.3596 0>}
  MaterialFunctionCall(Vector3=Divide(A=Multiply(A=@n15, B=@n16.RGB), B=Multiply(A=@n8, B=@n14.RGB))) {materialFunction=&DebugFloat3Values}
  n17 = TextureSample!(UVs, Tex, Apply View MipBias) {samplerType=SAMPLERTYPE_LinearColor, texture=&T_Athena_Grass_Farm_ColorMatched_D}
  Multiply(A=VectorParameter() {parameterName=NewTint2, defaultValue=<rgba 0.5168 0.3932 0.0519 0>}.RGB, B=@n17.RGB)
  n18 = TextureSample!(UVs, Tex, Level, Apply View MipBias) {constMipValue=0, mipValueMode=TMVM_MipLevel, texture=&T_Grasslands_D}
  Multiply(A=VectorParameter() {parameterName=NewTint3, defaultValue=<rgba 0.3913 0.4533 1.0287 0>}.RGB, B=@n18.RGB)
  OneMinus(None=@n10.A)
  ScalarParameter() {defaultValue=12}
  TextureSample!(UVs, Tex, Apply View MipBias)
 ]
}
